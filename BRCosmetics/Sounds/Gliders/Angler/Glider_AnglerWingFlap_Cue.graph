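# Sound cue for the wing flap of the Angler glider cosmetic.
# A random node picks one of four equally weighted flap recordings, played quietly and pitched down.
# It uses the player glider thrust attenuation and the glider landing sound class.
# A concurrency override, which takes precedence over the assigned open/close concurrency set, caps voices per owner and stops the oldest.
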SoundCue /BRCosmetics/Sounds/Gliders/Angler/Glider_AnglerWingFlap_Cue
root=Random_1
nodes [
 {id=Random_1, children [WavePlayer_2, WavePlayer_7, WavePlayer_8, WavePlayer_10], properties {weights [1,1,1,1]}}
 {id=WavePlayer_10, properties {soundWaveAssetPtr=&/Glider_AnglerWingFlap_06}}
 {id=WavePlayer_2, properties {soundWaveAssetPtr=&/Glider_AnglerWingFlap_02}}
 {id=WavePlayer_7, properties {soundWaveAssetPtr=&/Glider_AnglerWingFlap_03}}
 {id=WavePlayer_8, properties {soundWaveAssetPtr=&/Glider_AnglerWingFlap_04}}
]
properties {
 attenuationSettings=&Fort_Player_Glider_Thrust_Athena_Att
 bOverrideConcurrency=true
 maxDistance=12500
 pitchMultiplier=0.8
 soundClassObject=&Fort_Gliders_Land
 volumeMultiplier=0.1
 concurrencyOverrides {bLimitToOwner=true, maxCount=2, resolutionRule=StopOldest}
 concurrencySet [&Athena_GliderOpenClose_Conc]
}
